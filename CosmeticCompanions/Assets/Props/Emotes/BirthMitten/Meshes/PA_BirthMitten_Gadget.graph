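# Physics asset for the BirthMitten gadget, an emote prop mesh with two arms, armrests and boosters, in the CosmeticCompanions content.
# Each of its thirteen bones gets a single capsule body, mirrored left and right around the central cog body.
# Shoulders, armrests and boosters constrain to the cog, and each arm chains from shoulder through arm and lower arm to hand.
# All joints share the same symmetric swing and twist limits.
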PhysicsAsset /CosmeticCompanions/Assets/Props/Emotes/BirthMitten/Meshes/PA_BirthMitten_Gadget
bodies {
 arm_left [{shapeType=Capsule, length=30.2844, radius=10.2117, center=<19.8262 -0.1079 -0.6094>, rotation=<rot 0 -90.2792 -88.2709>}]
 arm_right [{shapeType=Capsule, length=30.2883, radius=10.2118, center=<-19.8295 0.1281 0.5549>, rotation=<rot 0 -90.3374 -88.4281>}]
 armrest_left [{shapeType=Capsule, length=23.3043, radius=8.1447, center=<7.0432 18.8422 0.5545>, rotation=<rot 0 174.9408 -88.7546>}]
 armrest_right [{shapeType=Capsule, length=23.3043, radius=8.1447, center=<-7.0432 18.8422 0.5545>, rotation=<rot 0 -174.9408 -88.7546>}]
 booster_left [{shapeType=Capsule, length=37.3568, radius=19.5247, center=<11.1481 -7.3299 -15.4138>, rotation=<rot 26.2726 -90 -3.5736>}]
 booster_right [{shapeType=Capsule, length=37.3568, radius=19.5247, center=<-11.1481 7.3299 15.4138>, rotation=<rot 26.2726 -90 -3.5736>}]
 cog [{shapeType=Capsule, length=57.7117, radius=29.9131, center=<-0.0034 -0.6871 26.2112>, rotation=<rot -24.3254 -90 -0.0099>}]
 hand_left [{shapeType=Capsule, length=30.3324, radius=29.4355, center=<-0.1835 -3.0786 29.6766>, rotation=<rot 24.7382 -90 -1.8133>}]
 hand_right [{shapeType=Capsule, length=30.3323, radius=29.4355, center=<0.1835 3.0786 -29.6766>, rotation=<rot 24.7381 -90 -1.8133>}]
 lowerarm_left [{shapeType=Capsule, length=22.1913, radius=12.8717, center=<17.452 -1.286 0.3979>, rotation=<rot 0 -90.6181 -89.1519>}]
 lowerarm_right [{shapeType=Capsule, length=22.1913, radius=12.8717, center=<-17.452 1.286 -0.3979>, rotation=<rot 0 -90.6181 -89.1518>}]
 shoulder_left [{shapeType=Capsule, length=43.0219, radius=38.0066, center=<37.7021 9.9017 13.9374>, rotation=<rot 0 120.8941 -85.0579>}]
 shoulder_right [{shapeType=Capsule, length=43.0302, radius=38.0001, center=<-37.6659 -9.8838 -13.9324>, rotation=<rot 0 120.6445 -85.0551>}]
}
constraints [
 {bone1Name=armrest_left, bone2Name=cog, swing1LimitDegrees=45, swing2LimitDegrees=45, twistLimitDegrees=45}
 {bone1Name=booster_left, bone2Name=cog, swing1LimitDegrees=45, swing2LimitDegrees=45, twistLimitDegrees=45}
 {bone1Name=armrest_right, bone2Name=cog, swing1LimitDegrees=45, swing2LimitDegrees=45, twistLimitDegrees=45}
 {bone1Name=booster_right, bone2Name=cog, swing1LimitDegrees=45, swing2LimitDegrees=45, twistLimitDegrees=45}
 {bone1Name=shoulder_left, bone2Name=cog, swing1LimitDegrees=45, swing2LimitDegrees=45, twistLimitDegrees=45}
 {bone1Name=arm_left, bone2Name=shoulder_left, swing1LimitDegrees=45, swing2LimitDegrees=45, twistLimitDegrees=45}
 {bone1Name=lowerarm_left, bone2Name=arm_left, swing1LimitDegrees=45, swing2LimitDegrees=45, twistLimitDegrees=45}
 {bone1Name=hand_left, bone2Name=lowerarm_left, swing1LimitDegrees=45, swing2LimitDegrees=45, twistLimitDegrees=45}
 {bone1Name=shoulder_right, bone2Name=cog, swing1LimitDegrees=45, swing2LimitDegrees=45, twistLimitDegrees=45}
 {bone1Name=arm_right, bone2Name=shoulder_right, swing1LimitDegrees=45, swing2LimitDegrees=45, twistLimitDegrees=45}
 {bone1Name=lowerarm_right, bone2Name=arm_right, swing1LimitDegrees=45, swing2LimitDegrees=45, twistLimitDegrees=45}
 {bone1Name=hand_right, bone2Name=lowerarm_right, swing1LimitDegrees=45, swing2LimitDegrees=45, twistLimitDegrees=45}
]
